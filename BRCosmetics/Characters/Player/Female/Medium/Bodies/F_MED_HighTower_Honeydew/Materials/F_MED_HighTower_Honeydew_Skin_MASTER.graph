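# Master skin material for the Honeydew HighTower female character, with a parameter-driven transition that recolors the skin.
# Base character PreFX feeds switchable cloth fuzz, silk and clear coat layers, then Honeydew and optional holo-foil, then PostFX.
# World-aligned noise, projected using pre-skinned normals, drives the dissolve between original and hue/saturation-shifted color, altered normals and an emissive edge, optionally limited by the SkinFX mask.
# A separate intensity parameter reveals tinted veins through a mask.
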Material /BRCosmetics/Characters/Player/Female/Medium/Bodies/F_MED_HighTower_Honeydew/Materials/F_MED_HighTower_Honeydew_Skin_MASTER
graph {
 properties {
  bAutomaticallySetUsageInEditor=false
  bUseMaterialAttributes=true
  bUsedWithMeshDeformer=true
  bUsedWithMorphTargets=true
  bUsedWithSkeletalMesh=true
  pixelDepthOffsetMode=PDOM_Legacy
  shadingModel=MSM_FromMaterialExpression
  subsurfaceProfile=&SS_HeroSkin_02
  usedShadingModels=Default Lit | Clear Coat | Subsurface Profile | Cloth
 }
 nodes [
  n1 = MaterialFunctionCall() {materialFunction=&BaseFN_CharacterFunction_PreFX}
  n2 = VertexInterpolator(VS=PreSkinnedNormal())
  n3 = TextureSampleParameter2D() {parameterName=SkinFX_Mask, samplerType=SAMPLERTYPE_Masks, texture=&T_BlackMask}
  n4 = StaticSwitchParameter(
   True=MaterialFunctionCall!(
    Alpha=StaticComponentMaskParameter(None=@n3.RGBA) {defaultR=true, group=Skin FX, parameterName=Cloth Channel}
    Gmap=@n1.Gmap
    Material=@n1.Gmap
   ) {materialFunction=/Game/Weapons/WeaponSkins/MaterialFunction/ClothFuzz}.Result
   False=@n1.Result
  ) {group=Skin FX, parameterName=Use Cloth fuzz}
  n5 = StaticSwitchParameter(
   True=MaterialFunctionCall!(Alpha=ChannelMaskParameter(None=@n3.RGBA) {group=Skin FX, parameterName=Silk Channel}.R, Material=@n4) {materialFunction=/Game/Weapons/WeaponSkins/MaterialFunction/MF_Silk}.Result
   False=@n4
  ) {group=Skin FX, parameterName=Use Silk}
  n6 = StaticSwitchParameter(
   True=MaterialFunctionCall!(Alpha=StaticComponentMaskParameter(None=@n3.RGBA) {defaultR=true, group=Skin FX, parameterName=Clear Coat Channel}, In=@n5) {materialFunction=/Game/Weapons/WeaponSkins/MaterialFunction/Clearcoat}.Result
   False=@n5
  ) {group=Skin FX, parameterName=Use Clear Coat}
  n7 = GetMaterialAttributes(None=@n6) {attributeGetTypes [#1, #2, #3]}
  n8 = ComponentMask(None=@n7.Base Color) {g=true}
  n9 = MaterialFunctionCall(Hue Shift Percentage=ScalarParameter() {group=HoneyDew, parameterName=Hue}, Texture=@n7.Base Color) {materialFunction=&HueShift}
  n10 = ScalarParameter() {defaultValue=1, group=HoneyDew, parameterName=Saturation}
  n11 = TextureObjectParameter() {
   group=HoneyDew
   parameterName=noisetex
   samplerType=SAMPLERTYPE_LinearColor
   texture=/Game/Characters/Player/Male/Medium/Bodies/M_MED_SRDrift/Textures/T_ElectricNoise
  }
  n12 = MaterialFunctionCall(TextureObject=@n11, TextureSize=ScalarParameter() {defaultValue=64, group=HoneyDew, parameterName=TexSize}, World Space Normal=@n2.PS) {materialFunction=&WorldAlignedTexture}
  n13 = ScalarParameter() {group=HoneyDew, parameterName=TransitionAmt}
  n14 = LinearInterpolate(A=OneMinus(None=@n13), B=@n13, Alpha=ScalarParameter() {group=HoneyDew, parameterName=ReverseTransition})
  n15 = Power(Base=@n14, Exp=ScalarParameter() {defaultValue=1, group=HoneyDew, parameterName=TransitionPow})
  n16 = Multiply(A=Saturate(None=@n15)) {constB=2}
  n17 = Saturate(None=MaterialFunctionCall(Alpha=@n12.XYZ Texture, Min=Subtract(A=@n16), Max=@n16) {materialFunction=&SmoothStep}.Result)
  n18 = LinearInterpolate(
   A=Constant() {r=1}
   B=StaticComponentMaskParameter(None=@n3.RGBA) {defaultB=true, group=Skin FX, parameterName=HoneyDew Skin Channel}
   Alpha=Round(None=ScalarParameter() {group=HoneyDew, parameterName=HoneyDewUsesFxMask})
  )
  n19 = LinearInterpolate(A=@n7.Base Color, B=LinearInterpolate(A=@n7.Base Color, B=LinearInterpolate(A=@n8, B=@n9.Result, Alpha=@n10), Alpha=@n17), Alpha=@n18)
  n20 = Multiply(A=@n19, B=VectorParameter() {group=HoneyDew, parameterName=VeinsColorMul, defaultValue=<rgba 0 0.1198 0 0>}.RGB)
  n21 = TextureSampleParameter2D() {group=HoneyDew, parameterName=VeinsMask, samplerType=SAMPLERTYPE_Masks, texture=/Engine/MaterialTemplates/LayerBlends/T_DefaultMaskTexture}
  n22 = ScalarParameter() {group=HoneyDew, parameterName=BP_VeinsIntensity}
  n23 = Saturate(None=Subtract(A=@n21.R, B=OneMinus(None=@n22)))
  n24 = MaterialFunctionCall(Float3=@n7.Normal) {materialFunction=&BreakOutFloat3Components}
  n25 = Multiply(
   A=AppendVector(A=@n24.R, B=@n24.R)
   B=Power(Base=@n14, Exp=ScalarParameter() {defaultValue=1, group=HoneyDew, parameterName=NormalsTransitionPow})
  )
  n26 = LinearInterpolate(A=@n7.Normal, B=Normalize(VectorInput=AppendVector(A=@n25, B=@n24.B)), Alpha=@n17)
  n27 = Multiply(A=@n15) {constB=3}
  n28 = Power(Base=Distance(A=@n12.XYZ Texture, B=Subtract(A=@n27)), Exp=ScalarParameter() {group=HoneyDew, parameterName=EmissiveEdgePow}) {constExponent=10}
  n29 = Multiply(A=Multiply(A=OneMinus(None=@n28), B=VectorParameter() {group=HoneyDew, parameterName=EmissiveEdge, defaultValue=<rgba 0 1 0 0>}.RGB), B=@n18)
  n30 = StaticSwitchParameter(
   True=SetMaterialAttributes!(
    MaterialAttributes=@n7.MaterialAttributes
    Base Color=LinearInterpolate(A=@n19, B=@n20, Alpha=@n23)
    Normal=LinearInterpolate(A=@n7.Normal, B=@n26, Alpha=@n18)
    Emissive Color=Add(A=@n7.Emissive Color, B=@n29)
   ) {attributeSetTypes [#1, #2, #3]}
   False=@n6
  ) {defaultValue=true, parameterName=UseHoneDewTech}
  MaterialFunctionCall(
   AO=@n1.AO
   Distance Blend=@n1.AO
   Fresnel=@n1.AO
   Gmap=@n1.AO
   Material=StaticSwitchParameter(
    True=MaterialFunctionCall!(PreskinnedLocalPosition=@n1.Pre skinned Local Position, PreskinnedLocalNormal=@n2.PS, Material=@n30) {materialFunction=/Game/Characters/Player/Female/Medium/Bodies/F_MED_HighTower_Squash/Materials/MF_HoloFoil}.Result
    False=@n30
   ) {parameterName=BUseHoloFoil}
  ) {materialFunction=&BaseFN_CharacterFunction_PostFX}
 ]
}
