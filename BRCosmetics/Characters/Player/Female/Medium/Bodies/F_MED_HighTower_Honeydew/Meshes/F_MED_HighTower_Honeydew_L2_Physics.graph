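# Physics asset for the Honeydew HighTower female medium body (L2 mesh), setting up physics for the left-side dynamic skirt bones.
# Capsule bodies cover front and back skirt chains, a thigh skirt piece and both thighs.
# Each chain's small base body rotates freely against the left thigh, while the two segments below bend on one limited swing axis.
# The thigh piece swings narrowly on the other axis; the right thigh has no constraint.
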PhysicsAsset /BRCosmetics/Characters/Player/Female/Medium/Bodies/F_MED_HighTower_Honeydew/Meshes/F_MED_HighTower_Honeydew_L2_Physics
bodies {
 dyn_skirt_bk_a_base_l [{shapeType=Capsule, length=1, radius=1, shapeName=dyn_skirt_bk_a_base_l_capsule}]
 dyn_skirt_bk_a_l [{shapeType=Capsule, length=5.3, radius=7, shapeName=dyn_skirt_bk_a_l_capsule, center=<10 2.5 0>, rotation=<rot 90 0 0>}]
 dyn_skirt_bk_b_l [{shapeType=Capsule, length=9.5, radius=5.5, shapeName=dyn_skirt_bk_b_l_capsule, center=<10 2.5 0>, rotation=<rot 90 0 0>}]
 dyn_skirt_fr_a_base_l [{shapeType=Capsule, length=1, radius=1, shapeName=dyn_skirt_fr_a_base_l_capsule}]
 dyn_skirt_fr_a_l [{shapeType=Capsule, length=7, radius=5.5, shapeName=dyn_skirt_fr_a_l_capsule, center=<10 -2 0>, rotation=<rot 90 0 0>}]
 dyn_skirt_fr_b_l [{shapeType=Capsule, length=8, radius=5.5, shapeName=dyn_skirt_fr_b_l_capsule, center=<10 -2 0>, rotation=<rot 90 0 0>}]
 dyn_skirt_thigh_l [{shapeType=Capsule, length=11, radius=4.5, shapeName=dyn_skirt_thigh_l_capsule, center=<-10 0 -6>, rotation=<rot 90 0 0>}]
 thigh_l [{shapeType=Capsule, length=31, radius=5.5, shapeName=thigh_l_capsule, center=<-19 -0.6 1>, rotation=<rot 90 0 0>}]
 thigh_r [{shapeType=Capsule, length=31, radius=5.5, shapeName=thigh_r_capsule, center=<19 0.6 -1>, rotation=<rot 90 0 0>}]
}
constraints [
 {
  bone1Name=dyn_skirt_fr_a_base_l
  bone2Name=thigh_l
  swing1LimitDegrees=45
  swing1Motion=Free
  swing2LimitDegrees=45
  swing2Motion=Free
  twistLimitDegrees=45
  twistMotion=Free
 }
 {
  bone1Name=dyn_skirt_fr_a_l
  bone2Name=dyn_skirt_fr_a_base_l
  swing1LimitDegrees=10
  swing2LimitDegrees=45
  swing2Motion=Locked
  twistLimitDegrees=45
  twistMotion=Locked
 }
 {
  bone1Name=dyn_skirt_fr_b_l
  bone2Name=dyn_skirt_fr_a_l
  swing1LimitDegrees=10
  swing2LimitDegrees=45
  swing2Motion=Locked
  twistLimitDegrees=45
  twistMotion=Locked
 }
 {
  bone1Name=dyn_skirt_bk_a_base_l
  bone2Name=thigh_l
  swing1LimitDegrees=45
  swing1Motion=Free
  swing2LimitDegrees=45
  swing2Motion=Free
  twistLimitDegrees=45
  twistMotion=Free
 }
 {
  bone1Name=dyn_skirt_bk_a_l
  bone2Name=dyn_skirt_bk_a_base_l
  swing1LimitDegrees=15
  swing2LimitDegrees=45
  swing2Motion=Locked
  twistLimitDegrees=45
  twistMotion=Locked
 }
 {
  bone1Name=dyn_skirt_bk_b_l
  bone2Name=dyn_skirt_bk_a_l
  swing1LimitDegrees=15
  swing2LimitDegrees=45
  swing2Motion=Locked
  twistLimitDegrees=45
  twistMotion=Locked
 }
 {
  bone1Name=dyn_skirt_thigh_l
  bone2Name=thigh_l
  swing1LimitDegrees=45
  swing1Motion=Locked
  swing2LimitDegrees=5
  twistLimitDegrees=45
  twistMotion=Locked
 }
]
