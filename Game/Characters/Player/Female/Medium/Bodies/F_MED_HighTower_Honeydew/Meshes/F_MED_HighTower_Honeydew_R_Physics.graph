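# Physics asset for the right-side dynamic skirt of the female medium HighTower Honeydew player body.
# Capsule bodies cover front, back and thigh skirt bones plus both thighs, with the left thigh left unconstrained.
# A two-link front chain, a single thigh piece and a back chain on a small free-rotating base attach to the right thigh.
# Other joints lock twist and one swing axis, so each bends in a limited single plane.
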PhysicsAsset /Game/Characters/Player/Female/Medium/Bodies/F_MED_HighTower_Honeydew/Meshes/F_MED_HighTower_Honeydew_R_Physics
bodies {
 dyn_skirt_bk_a_base_r [{shapeType=Capsule, length=1, radius=1}]
 dyn_skirt_bk_a_r [{shapeType=Capsule, length=5.3359, radius=7, center=<-10 -2.5 0>, rotation=<rot 90 0 180>}]
 dyn_skirt_bk_b_r [{shapeType=Capsule, length=9.579, radius=5.5, center=<-10 -2.5 0>, rotation=<rot 90 0 180>}]
 dyn_skirt_fr_a_r [{shapeType=Capsule, length=6.8554, radius=5.5, center=<-10 2 0>, rotation=<rot 90 0 180>}]
 dyn_skirt_fr_b_r [{shapeType=Capsule, length=7.9575, radius=5.5, center=<-10 2 0>, rotation=<rot 90 0 180>}]
 dyn_skirt_thigh_r [{shapeType=Capsule, length=11, radius=4.5, center=<10 0 6>, rotation=<rot 90 0 180>}]
 thigh_l [{shapeType=Capsule, length=31, radius=5.5, center=<-19.0092 -0.6001 1.0824>, rotation=<rot 87.6725 -180 -180>}]
 thigh_r [{shapeType=Capsule, length=31, radius=5.5, center=<19.0092 0.6001 -1.0824>, rotation=<rot 87.672 -180 0>}]
}
constraints [
 {
  bone1Name=dyn_skirt_fr_a_r
  bone2Name=thigh_r
  swing1LimitDegrees=10
  swing2LimitDegrees=20
  swing2Motion=Locked
  twistLimitDegrees=5
  twistMotion=Locked
 }
 {
  bone1Name=dyn_skirt_fr_b_r
  bone2Name=dyn_skirt_fr_a_r
  swing1LimitDegrees=10
  swing2LimitDegrees=20
  swing2Motion=Locked
  twistLimitDegrees=5
  twistMotion=Locked
 }
 {
  bone1Name=dyn_skirt_bk_b_r
  bone2Name=dyn_skirt_bk_a_r
  swing1LimitDegrees=15
  swing2LimitDegrees=5
  swing2Motion=Locked
  twistLimitDegrees=5
  twistMotion=Locked
 }
 {
  bone1Name=dyn_skirt_bk_a_r
  bone2Name=dyn_skirt_bk_a_base_r
  swing1LimitDegrees=15
  swing2LimitDegrees=5
  swing2Motion=Locked
  twistLimitDegrees=5
  twistMotion=Locked
 }
 {
  bone1Name=dyn_skirt_thigh_r
  bone2Name=thigh_r
  swing1LimitDegrees=45
  swing1Motion=Locked
  swing2LimitDegrees=5
  twistLimitDegrees=45
  twistMotion=Locked
 }
 {
  bone1Name=dyn_skirt_bk_a_base_r
  bone2Name=thigh_r
  swing1LimitDegrees=45
  swing1Motion=Free
  swing2LimitDegrees=45
  swing2Motion=Free
  twistLimitDegrees=45
  twistMotion=Free
 }
]
