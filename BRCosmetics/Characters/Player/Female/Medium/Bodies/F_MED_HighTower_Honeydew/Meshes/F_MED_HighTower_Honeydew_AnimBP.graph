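# Animation Blueprint for the female HighTower Honeydew cosmetic body, adding corrective and secondary motion to skirt, collar and thigh bones.
# It copies the parent mesh pose, applies constraints and additive bone offsets weighted by thigh, clavicle, head and crouch values, then runs collar AnimDynamics and thigh-relative RigidBody simulations.
# Final offsets reshape the skirt, some only while the locker idle montage plays.
# Several older-looking simulation and corrective chains remain disconnected.
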AnimBlueprint /BRCosmetics/Characters/Player/Female/Medium/Bodies/F_MED_HighTower_Honeydew/Meshes/F_MED_HighTower_Honeydew_AnimBP
target_skeleton=&/F_MED_HighTower_Honeydew_Skeleton
graphs [
 AnimGraph {nodes [
   n1 = Anim Dynamics - Bone: dyn_skirt_fr_a_l(Alpha=1, ExternalForce="0,0,0") {
    angularDampingOverride=1
    angularSpringConstant=1500
    bAngularSpring=true
    bGravityOverrideInSimSpace=true
    bOverrideAngularDamping=true
    gravityScale=0.35
    simulationSpace=RootRelative
    alphaBoolBlend {blendOption=Linear}
    boundBone {boneName=dyn_skirt_fr_a_l}
    physicsBodyDefinitions [{
      sphereCollisionRadius=0
      boundBone {boneName=dyn_skirt_fr_a_l}
      boxExtents=<10 20 20>
      constraintSetup {angularLimitsMax=<0 0 0>, angularLimitsMin=<0 0 -40>, angularTarget=<0 0 0>}
      localJointOffset=<5 0 0>
     }]
   }
   n2 = Anim Dynamics - Bone: dyn_skirt_fr_b_l(ComponentPose=@n1, Alpha=1, ExternalForce="0,0,0") {
    angularDampingOverride=1
    angularSpringConstant=1500
    bAngularSpring=true
    bGravityOverrideInSimSpace=true
    bOverrideAngularDamping=true
    gravityScale=0.35
    simulationSpace=RootRelative
    alphaBoolBlend {blendOption=Linear}
    boundBone {boneName=dyn_skirt_fr_b_l}
    physicsBodyDefinitions [{
      sphereCollisionRadius=0
      boundBone {boneName=dyn_skirt_fr_b_l}
      boxExtents=<10 20 20>
      constraintSetup {angularLimitsMax=<0 0 0>, angularLimitsMin=<0 0 -40>, angularTarget=<0 0 0>}
      localJointOffset=<5 0 0>
     }]
   }
   n3 = Anim Dynamics - Bone: dyn_skirt_bk_a_l(ComponentPose=@n2, Alpha=1, ExternalForce="0,0,0") {
    angularDampingOverride=1
    angularSpringConstant=1500
    bAngularSpring=true
    bGravityOverrideInSimSpace=true
    bOverrideAngularDamping=true
    gravityScale=0.35
    simulationSpace=RootRelative
    alphaBoolBlend {blendOption=Linear}
    boundBone {boneName=dyn_skirt_bk_a_l}
    physicsBodyDefinitions [{
      sphereCollisionRadius=0
      boundBone {boneName=dyn_skirt_bk_a_l}
      boxExtents=<10 20 20>
      constraintSetup {angularLimitsMax=<0 0 40>, angularLimitsMin=<0 0 0>, angularTarget=<0 0 0>}
      localJointOffset=<5 0 0>
     }]
   }
   n4 = Anim Dynamics - Bone: dyn_skirt_bk_b_l(ComponentPose=@n3, Alpha=1, ExternalForce="0,0,0") {
    angularDampingOverride=1
    angularSpringConstant=1500
    bAngularSpring=true
    bGravityOverrideInSimSpace=true
    bOverrideAngularDamping=true
    gravityScale=0.35
    simulationSpace=RootRelative
    alphaBoolBlend {blendOption=Linear}
    boundBone {boneName=dyn_skirt_bk_b_l}
    physicsBodyDefinitions [{
      sphereCollisionRadius=0
      boundBone {boneName=dyn_skirt_bk_b_l}
      boxExtents=<10 20 20>
      constraintSetup {angularLimitsMax=<0 0 40>, angularLimitsMin=<0 0 0>, angularTarget=<0 0 0>}
      localJointOffset=<5 0 0>
     }]
   }
   n5 = Anim Dynamics - Bone: dyn_skirt_thigh_l(ComponentPose=@n4, Alpha=1, ExternalForce="0,0,0") {
    angularDampingOverride=1
    angularSpringConstant=1500
    bAngularSpring=true
    bGravityOverrideInSimSpace=true
    bOverrideAngularDamping=true
    gravityScale=0.35
    simulationSpace=RootRelative
    alphaBoolBlend {blendOption=Linear}
    boundBone {boneName=dyn_skirt_thigh_l}
    physicsBodyDefinitions [{
      sphereCollisionRadius=0
      boundBone {boneName=dyn_skirt_thigh_l}
      boxExtents=<10 20 20>
      constraintSetup {angularLimitsMax=<0 0 0>, angularLimitsMin=<0 -15 0>, angularTarget=<0 0 0>}
      localJointOffset=<-5 0 0>
     }]
   }
   n6 = Anim Dynamics - Bone: dyn_skirt_fr_a_r(ComponentPose=@n5, Alpha=1, ExternalForce="0,0,0") {
    angularDampingOverride=1
    angularSpringConstant=1500
    bAngularSpring=true
    bGravityOverrideInSimSpace=true
    bOverrideAngularDamping=true
    gravityScale=0.35
    simulationSpace=RootRelative
    alphaBoolBlend {blendOption=Linear}
    boundBone {boneName=dyn_skirt_fr_a_r}
    physicsBodyDefinitions [{
      sphereCollisionRadius=0
      boundBone {boneName=dyn_skirt_fr_a_r}
      boxExtents=<10 20 20>
      constraintSetup {angularLimitsMax=<0 0 0>, angularLimitsMin=<0 0 -40>, angularTarget=<0 0 0>}
      localJointOffset=<-5 0 0>
     }]
   }
   n7 = Anim Dynamics - Bone: dyn_skirt_fr_b_r(ComponentPose=@n6, Alpha=1, ExternalForce="0,0,0") {
    angularDampingOverride=1
    angularSpringConstant=1500
    bAngularSpring=true
    bGravityOverrideInSimSpace=true
    bOverrideAngularDamping=true
    gravityScale=0.35
    simulationSpace=RootRelative
    alphaBoolBlend {blendOption=Linear}
    boundBone {boneName=dyn_skirt_fr_b_r}
    physicsBodyDefinitions [{
      sphereCollisionRadius=0
      boundBone {boneName=dyn_skirt_fr_b_r}
      boxExtents=<10 20 20>
      constraintSetup {angularLimitsMax=<0 0 0>, angularLimitsMin=<0 0 -40>, angularTarget=<0 0 0>}
      localJointOffset=<-5 0 0>
     }]
   }
   n8 = Anim Dynamics - Bone: dyn_skirt_bk_a_r(ComponentPose=@n7, Alpha=1, ExternalForce="0,0,0") {
    angularDampingOverride=1
    angularSpringConstant=1500
    bAngularSpring=true
    bGravityOverrideInSimSpace=true
    bOverrideAngularDamping=true
    gravityScale=0.35
    simulationSpace=RootRelative
    alphaBoolBlend {blendOption=Linear}
    boundBone {boneName=dyn_skirt_bk_a_r}
    physicsBodyDefinitions [{
      sphereCollisionRadius=0
      boundBone {boneName=dyn_skirt_bk_a_r}
      boxExtents=<10 20 20>
      constraintSetup {angularLimitsMax=<0 0 40>, angularLimitsMin=<0 0 0>, angularTarget=<0 0 0>}
      localJointOffset=<-5 0 0>
     }]
   }
   n9 = Anim Dynamics - Bone: dyn_skirt_bk_b_r(ComponentPose=@n8, Alpha=1, ExternalForce="0,0,0") {
    angularDampingOverride=1
    angularSpringConstant=1500
    bAngularSpring=true
    bGravityOverrideInSimSpace=true
    bOverrideAngularDamping=true
    gravityScale=0.35
    simulationSpace=RootRelative
    alphaBoolBlend {blendOption=Linear}
    boundBone {boneName=dyn_skirt_bk_b_r}
    physicsBodyDefinitions [{
      sphereCollisionRadius=0
      boundBone {boneName=dyn_skirt_bk_b_r}
      boxExtents=<10 20 20>
      constraintSetup {angularLimitsMax=<0 0 40>, angularLimitsMin=<0 0 0>, angularTarget=<0 0 0>}
      localJointOffset=<-5 0 0>
     }]
   }
   Anim Dynamics - Bone: dyn_skirt_thigh_r(ComponentPose=@n9, Alpha=1, ExternalForce="0,0,0") {
    angularDampingOverride=1
    angularSpringConstant=1500
    bAngularSpring=true
    bGravityOverrideInSimSpace=true
    bOverrideAngularDamping=true
    gravityScale=0.35
    simulationSpace=RootRelative
    alphaBoolBlend {blendOption=Linear}
    boundBone {boneName=dyn_skirt_thigh_r}
    physicsBodyDefinitions [{
      sphereCollisionRadius=0
      boundBone {boneName=dyn_skirt_thigh_r}
      boxExtents=<10 20 20>
      constraintSetup {angularLimitsMax=<0 0 0>, angularLimitsMin=<0 -15 0>, angularTarget=<0 0 0>}
      localJointOffset=<5 0 0>
     }]
   }
   n10 = "Constraint - dyn_skirt_fr_a_base (PP)"!(ComponentPose, Alpha=1, ConstraintWeights_0=0.5, ConstraintWeights_1:float=0.5) -> (Pose) {
    alphaBoolBlend {blendOption=Linear}
    boneToModify {boneName=dyn_skirt_fr_a_base}
    constraintSetup [{targetBone {boneName=dyn_skirt_fr_a_l}}, {targetBone {boneName=dyn_skirt_fr_a_r}}]
    constraintWeights [1,1]
   }
   "Constraint - dyn_skirt_bk_a_base (PP)"!(ComponentPose=@n10, Alpha=1, ConstraintWeights_0=0.5, ConstraintWeights_1:float=0.5) -> (Pose) {
    alphaBoolBlend {blendOption=Linear}
    boneToModify {boneName=dyn_skirt_bk_a_base}
    constraintSetup [{targetBone {boneName=dyn_skirt_bk_a_l}}, {targetBone {boneName=dyn_skirt_bk_a_r}}]
    constraintWeights [1,1]
   }
   n11 = Get ThighRotationLeft() -> (ThighRotationLeft:float)
   n12 = "Transform (Modify) Bone - Bone: dyn_skirt_fr_a_base_r"(Alpha=@n11, Translation="0,0,0", Rotation="0,-15,-15", Scale="1,1,1") {
    rotationMode=BMM_Additive
    rotationSpace=BCS_BoneSpace
    translationMode=BMM_Additive
    translationSpace=BCS_BoneSpace
    alphaBoolBlend {blendOption=Linear}
    alphaScaleBiasClamp {bMapRange=true, inRange {max=0, min=-100}, outRange {max=0, min=1}}
    boneToModify {boneName=dyn_skirt_fr_a_base_r}
    rotation {roll=-15, yaw=-15}
   }
   n13 = "Transform (Modify) Bone - Bone: dyn_skirt_bk_a_base_l"(ComponentPose=@n12, Alpha=@n11, Translation="-15,0,0", Rotation="0,10,30", Scale="1,1,1") {
    rotationMode=BMM_Additive
    rotationSpace=BCS_BoneSpace
    translationMode=BMM_Additive
    translationSpace=BCS_BoneSpace
    alphaBoolBlend {blendOption=Linear}
    alphaScaleBiasClamp {bMapRange=true, inRange {max=0, min=-100}, outRange {max=0, min=1}}
    boneToModify {boneName=dyn_skirt_bk_a_base_l}
    rotation {roll=30, yaw=10}
    translation {x=-15}
   }
   n14 = Get ThighRotationRight() -> (ThighRotationRight:float)
   n15 = "Transform (Modify) Bone - Bone: dyn_skirt_fr_a_base_l"(ComponentPose=@n13, Alpha=@n14, Translation="0,0,0", Rotation="0,-15,-15", Scale="1,1,1") {
    rotationMode=BMM_Additive
    rotationSpace=BCS_BoneSpace
    translationMode=BMM_Additive
    translationSpace=BCS_BoneSpace
    alphaBoolBlend {blendOption=Linear}
    alphaScaleBiasClamp {bMapRange=true, inRange {max=0, min=-100}, outRange {max=0, min=1}}
    boneToModify {boneName=dyn_skirt_fr_a_base_l}
    rotation {roll=-15, yaw=-15}
   }
   "Transform (Modify) Bone - Bone: dyn_skirt_bk_a_base_r"(ComponentPose=@n15, Alpha=@n14, Translation="10,0,0", Rotation="0,10,30", Scale="1,1,1") {
    rotationMode=BMM_Additive
    rotationSpace=BCS_BoneSpace
    translationMode=BMM_Additive
    translationSpace=BCS_BoneSpace
    alphaBoolBlend {blendOption=Linear}
    alphaScaleBiasClamp {bMapRange=true, inRange {max=0, min=-100}, outRange {max=0, min=1}}
    boneToModify {boneName=dyn_skirt_bk_a_base_r}
    rotation {roll=30, yaw=10}
    translation {x=10}
   }
   "Transform (Modify) Bone - Bone: thigh_volume_r"(Alpha=1, Translation="1,3,-1", Rotation="0,0,0", Scale="1,1,1") {
    translationMode=BMM_Additive
    translationSpace=BCS_BoneSpace
    alphaBoolBlend {blendOption=Linear}
    alphaScaleBiasClamp {bMapRange=true, inRange {max=0, min=-100}, outRange {max=0, min=1}}
    boneToModify {boneName=thigh_volume_r}
    translation=<1 3 -1>
   }
   n16 = "Transform (Modify) Bone - Bone: dyn_skirt_bk_a_base_l"(Alpha=0, Translation="0,3,-2", Rotation="0,0,0", Scale="1,1,1") {
    rotationMode=BMM_Additive
    rotationSpace=BCS_BoneSpace
    translationMode=BMM_Additive
    translationSpace=BCS_BoneSpace
    alphaBoolBlend {blendOption=Linear}
    alphaScaleBiasClamp {inRange {max=0, min=-100}, outRange {max=0, min=1}}
    boneToModify {boneName=dyn_skirt_bk_a_base_l}
    translation {y=3, z=-2}
   }
   n17 = "Transform (Modify) Bone - Bone: dyn_skirt_thigh_l"(
    ComponentPose=@n16
    Alpha=Get IsCrouchingAndStillAlpha() -> (IsCrouchingAndStillAlpha:float)
    Translation="0,0,0"
    Rotation="0,0,0"
    Scale="1,1,1"
   ) {
    rotationMode=BMM_Additive
    rotationSpace=BCS_BoneSpace
    translationMode=BMM_Additive
    translationSpace=BCS_BoneSpace
    alphaBoolBlend {blendOption=Linear}
    alphaScaleBiasClamp {inRange {max=0, min=-100}, outRange {max=0, min=1}}
    boneToModify {boneName=dyn_skirt_thigh_l}
   }
   n18 = Get IsCrouchingAndStillAlpha()
   n19 = "Transform (Modify) Bone - Bone: dyn_skirt_bk_a_base_r"(ComponentPose=@n17, Alpha=@n18, Translation="15,-5,0", Rotation="0,-15,0", Scale="1,1,1") {
    rotationMode=BMM_Additive
    rotationSpace=BCS_BoneSpace
    translationMode=BMM_Additive
    translationSpace=BCS_BoneSpace
    alphaBoolBlend {blendOption=Linear}
    alphaScaleBiasClamp {inRange {max=0, min=-100}, outRange {max=0, min=1}}
    boneToModify {boneName=dyn_skirt_bk_a_base_r}
    rotation {yaw=-15}
    translation=<15 -5>
   }
   "Transform (Modify) Bone - Bone: dyn_skirt_thigh_r"(ComponentPose=@n19, Alpha=@n18, Translation="-3,0,0", Rotation="0,0,0", Scale="1,1,1") {
    rotationMode=BMM_Additive
    rotationSpace=BCS_BoneSpace
    translationMode=BMM_Additive
    translationSpace=BCS_BoneSpace
    alphaBoolBlend {blendOption=Linear}
    alphaScaleBiasClamp {inRange {max=0, min=-100}, outRange {max=0, min=1}}
    boneToModify {boneName=dyn_skirt_thigh_r}
    translation {x=-3}
   }
   n20 = Get SkirtDynAlpha() -> (SkirtDynAlpha:float)
   n21 = RigidBody(Alpha=@n20) {
    bClampLinearTranslationLimitToRefPose=true
    bOverrideWorldGravity=true
    overridePhysicsAsset=/Game/Characters/Player/Female/Medium/Bodies/F_MED_HighTower_Honeydew/Meshes/F_MED_HighTower_Honeydew_L_Physics
    simulationSpace=BaseBoneSpace
    alphaBoolBlend {blendOption=Linear}
    baseBoneRef {boneName=thigh_l}
    componentAppliedLinearAccClamp=<3000 3000 3000>
    componentLinearAccScale=<0.3 0.3 0.3>
    componentLinearVelScale=<0.3 0.3 0.3>
    overrideWorldGravity {z=-980}
   }
   RigidBody(ComponentPose=@n21, Alpha=@n20) {
    bClampLinearTranslationLimitToRefPose=true
    bOverrideWorldGravity=true
    overridePhysicsAsset=/Game/Characters/Player/Female/Medium/Bodies/F_MED_HighTower_Honeydew/Meshes/F_MED_HighTower_Honeydew_R_Physics
    simulationSpace=BaseBoneSpace
    alphaBoolBlend {blendOption=Linear}
    baseBoneRef {boneName=thigh_r}
    componentAppliedLinearAccClamp=<3000 3000 3000>
    componentLinearAccScale=<0.3 0.3 0.3>
    componentLinearVelScale=<0.3 0.3 0.3>
    overrideWorldGravity {z=-980}
   }
   RigidBody(Alpha=Get SkirtDynAlpha()) {
    bClampLinearTranslationLimitToRefPose=true
    bOverrideWorldGravity=true
    overridePhysicsAsset=/Game/Characters/Player/Female/Medium/Bodies/F_MED_HighTower_Honeydew/Meshes/F_MED_HighTower_Honeydew_Physics
    simulationSpace=BaseBoneSpace
    alphaBoolBlend {blendOption=Linear}
    baseBoneRef {boneName=pelvis}
    componentLinearAccScale=<0.3 0.3 0.3>
    componentLinearVelScale=<0.3 0.3 0.3>
    overrideWorldGravity {z=-980}
   }
   n22 = "Constraint - thigh_volume_l (R)"(ComponentPose=Local To Component(LocalPose=Copy Pose From Mesh() {bUseAttachedParent=true}), Alpha=0.5, ConstraintWeights_0=1) {
    alphaBoolBlend {blendOption=Linear}
    boneToModify {boneName=thigh_volume_l}
    constraintSetup [{transformType=Rotation, targetBone {boneName=thigh_l}}]
    constraintWeights [1]
   }
   n23 = "Constraint - thigh_volume_r (R)"(ComponentPose=@n22, Alpha=0.5, ConstraintWeights_0=1) {
    alphaBoolBlend {blendOption=Linear}
    boneToModify {boneName=thigh_volume_r}
    constraintSetup [{transformType=Rotation, targetBone {boneName=thigh_r}}]
    constraintWeights [1]
   }
   n24 = "Transform (Modify) Bone - Bone: thigh_volume_l"(ComponentPose=@n23, Alpha=Get ThighRotationLeft(), Translation="-3,-5,-1", Rotation="0,0,0", Scale="1,1,1") {
    rotationMode=BMM_Additive
    rotationSpace=BCS_BoneSpace
    scaleMode=BMM_Additive
    scaleSpace=BCS_BoneSpace
    translationMode=BMM_Additive
    translationSpace=BCS_BoneSpace
    alphaBoolBlend {blendOption=Linear}
    alphaScaleBiasClamp {bMapRange=true, inRange {max=0, min=-100}, outRange {max=0, min=1}}
    boneToModify {boneName=thigh_volume_l}
    translation=<-3 -5 -1>
   }
   n25 = "Transform (Modify) Bone - Bone: thigh_volume_r"(ComponentPose=@n24, Alpha=Get ThighRotationRight(), Translation="3,5,1", Rotation="0,0,0", Scale="1,1,1") {
    translationMode=BMM_Additive
    translationSpace=BCS_BoneSpace
    alphaBoolBlend {blendOption=Linear}
    alphaScaleBiasClamp {bMapRange=true, inRange {max=0, min=-100}, outRange {max=0, min=1}}
    boneToModify {boneName=thigh_volume_r}
    translation=<3 5 1>
   }
   n26 = Get ClavicleLeft() -> (ClavicleLeft:float)
   n27 = "Transform (Modify) Bone - Bone: dyn_collar_fr_b_l"(ComponentPose=@n25, Alpha=@n26, Translation="0,0,0", Rotation="30,0,0", Scale="1,1,1") {
    rotationMode=BMM_Additive
    rotationSpace=BCS_BoneSpace
    translationMode=BMM_Additive
    translationSpace=BCS_BoneSpace
    alphaBoolBlend {blendOption=Linear}
    alphaScaleBiasClamp {bMapRange=true, inRange {max=50}}
    boneToModify {boneName=dyn_collar_fr_b_l}
    rotation {pitch=30}
   }
   n28 = "Transform (Modify) Bone - Bone: dyn_collar_side_a_l"(ComponentPose=@n27, Alpha=@n26, Translation="0,0,0", Rotation="10,0,0", Scale="1,1,1") {
    rotationMode=BMM_Additive
    rotationSpace=BCS_BoneSpace
    translationMode=BMM_Additive
    translationSpace=BCS_BoneSpace
    alphaBoolBlend {blendOption=Linear}
    alphaScaleBiasClamp {bMapRange=true, inRange {max=50}}
    boneToModify {boneName=dyn_collar_side_a_l}
    rotation {pitch=10}
   }
   n29 = "Transform (Modify) Bone - Bone: dyn_collar_side_b_l"(ComponentPose=@n28, Alpha=@n26, Translation="0,0,0", Rotation="110,20,0", Scale="1,1,1") {
    rotationMode=BMM_Additive
    rotationSpace=BCS_BoneSpace
    translationMode=BMM_Additive
    translationSpace=BCS_BoneSpace
    alphaBoolBlend {blendOption=Linear}
    alphaScaleBiasClamp {bMapRange=true, inRange {max=50}}
    boneToModify {boneName=dyn_collar_side_b_l}
    rotation {pitch=110, yaw=20}
   }
   n30 = Get ClavicleRight() -> (ClavicleRight:float)
   n31 = "Transform (Modify) Bone - Bone: dyn_collar_fr_b_r"(ComponentPose=@n29, Alpha=@n30, Translation="0,0,0", Rotation="30,0,0", Scale="1,1,1") {
    rotationMode=BMM_Additive
    rotationSpace=BCS_BoneSpace
    translationMode=BMM_Additive
    translationSpace=BCS_BoneSpace
    alphaBoolBlend {blendOption=Linear}
    alphaScaleBiasClamp {bMapRange=true, inRange {max=50}}
    boneToModify {boneName=dyn_collar_fr_b_r}
    rotation {pitch=30}
   }
   n32 = "Transform (Modify) Bone - Bone: dyn_collar_side_a_r"(ComponentPose=@n31, Alpha=@n30, Translation="0,0,0", Rotation="10,0,0", Scale="1,1,1") {
    rotationMode=BMM_Additive
    rotationSpace=BCS_BoneSpace
    translationMode=BMM_Additive
    translationSpace=BCS_BoneSpace
    alphaBoolBlend {blendOption=Linear}
    alphaScaleBiasClamp {bMapRange=true, inRange {max=50}}
    boneToModify {boneName=dyn_collar_side_a_r}
    rotation {pitch=10}
   }
   n33 = "Transform (Modify) Bone - Bone: dyn_collar_side_b_r"(ComponentPose=@n32, Alpha=@n30, Translation="0,0,0", Rotation="110,20,0", Scale="1,1,1") {
    rotationMode=BMM_Additive
    rotationSpace=BCS_BoneSpace
    translationMode=BMM_Additive
    translationSpace=BCS_BoneSpace
    alphaBoolBlend {blendOption=Linear}
    alphaScaleBiasClamp {bMapRange=true, inRange {max=50}}
    boneToModify {boneName=dyn_collar_side_b_r}
    rotation {pitch=110, yaw=20}
   }
   n34 = "Transform (Modify) Bone - Bone: dyn_collar_bk_a"(ComponentPose=@n33, Alpha=Get HeadRotation() -> (HeadRotation:float), Translation="0,0,0", Rotation="-20,0,0", Scale="1,1,1") {
    rotationMode=BMM_Additive
    rotationSpace=BCS_BoneSpace
    translationMode=BMM_Additive
    translationSpace=BCS_BoneSpace
    alphaBoolBlend {blendOption=Linear}
    alphaScaleBiasClamp {bMapRange=true, inRange {max=0, min=-25}, outRange {max=0, min=1}}
    boneToModify {boneName=dyn_collar_bk_a}
    rotation {pitch=-20}
   }
   n35 = "Constraint - dyn_skirt_fr_a_base_l (P)"(ComponentPose=@n34, Alpha=1, ConstraintWeights_0=1) {
    alphaBoolBlend {blendOption=Linear}
    boneToModify {boneName=dyn_skirt_fr_a_base_l}
    constraintSetup [{targetBone {boneName=thigh_twist_01_l}}]
    constraintWeights [1]
   }
   n36 = "Constraint - dyn_skirt_fr_a_base_r (P)"(ComponentPose=@n35, Alpha=1, ConstraintWeights_0=1) {
    alphaBoolBlend {blendOption=Linear}
    boneToModify {boneName=dyn_skirt_fr_a_base_r}
    constraintSetup [{targetBone {boneName=thigh_twist_01_r}}]
    constraintWeights [1]
   }
   n37 = "Constraint - dyn_skirt_bk_a_base_l (PP)"!(ComponentPose=@n36, Alpha=1, ConstraintWeights_0=0.5, ConstraintWeights_1:float=0.5) -> (Pose) {
    alphaBoolBlend {blendOption=Linear}
    boneToModify {boneName=dyn_skirt_bk_a_base_l}
    constraintSetup [{targetBone {boneName=thigh_twist_01_l}}, {targetBone {boneName=pelvis}}]
    constraintWeights [1,1]
   }
   n38 = "Constraint - dyn_skirt_bk_a_base_r (PP)"!(ComponentPose=@n37, Alpha=1, ConstraintWeights_0=0.5, ConstraintWeights_1:float=0.5) -> (Pose) {
    alphaBoolBlend {blendOption=Linear}
    boneToModify {boneName=dyn_skirt_bk_a_base_r}
    constraintSetup [{targetBone {boneName=thigh_twist_01_r}}, {targetBone {boneName=pelvis}}]
    constraintWeights [1,1]
   }
   n39 = "Constraint - dyn_skirt_fr_a_base (PP)"(ComponentPose=@n38, Alpha=1, ConstraintWeights_0=0.5, ConstraintWeights_1=0.5) {
    alphaBoolBlend {blendOption=Linear}
    boneToModify {boneName=dyn_skirt_fr_a_base}
    constraintSetup [{targetBone {boneName=dyn_skirt_fr_a_l}}, {targetBone {boneName=dyn_skirt_fr_a_r}}]
    constraintWeights [1,1]
   }
   n40 = "Constraint - dyn_skirt_bk_a_base (PP)"(ComponentPose=@n39, Alpha=1, ConstraintWeights_0=0.5, ConstraintWeights_1=0.5) {
    alphaBoolBlend {blendOption=Linear}
    boneToModify {boneName=dyn_skirt_bk_a_base}
    constraintSetup [{targetBone {boneName=dyn_skirt_bk_a_l}}, {targetBone {boneName=dyn_skirt_bk_a_r}}]
    constraintWeights [1,1]
   }
   n41 = "Constraint - dyn_skirt_fr_b (PP)"!(ComponentPose=@n40, Alpha=1, ConstraintWeights_0=0.5, ConstraintWeights_1:float=0.5) -> (Pose) {
    alphaBoolBlend {blendOption=Linear}
    boneToModify {boneName=dyn_skirt_fr_b}
    constraintSetup [{targetBone {boneName=dyn_skirt_fr_b_l}}, {targetBone {boneName=dyn_skirt_fr_b_r}}]
    constraintWeights [1,1]
   }
   n42 = "Constraint - dyn_skirt_bk_b (PP)"!(ComponentPose=@n41, Alpha=1, ConstraintWeights_0=0.5, ConstraintWeights_1:float=0.5) -> (Pose) {
    alphaBoolBlend {blendOption=Linear}
    boneToModify {boneName=dyn_skirt_bk_b}
    constraintSetup [{targetBone {boneName=dyn_skirt_bk_b_l}}, {targetBone {boneName=dyn_skirt_bk_b_r}}]
    constraintWeights [1,1]
   }
   n43 = Get ThighRotationLeft()
   n44 = "Transform (Modify) Bone - Bone: dyn_skirt_fr_a_base_l"(ComponentPose=@n42, Alpha=@n43, Translation="0,0,0", Rotation="-5,0,0", Scale="1,1,1") {
    rotationMode=BMM_Additive
    rotationSpace=BCS_BoneSpace
    translationMode=BMM_Additive
    translationSpace=BCS_BoneSpace
    alphaBoolBlend {blendOption=Linear}
    alphaScaleBiasClamp {bMapRange=true, inRange {max=0, min=-100}, outRange {max=0, min=1}}
    boneToModify {boneName=dyn_skirt_fr_a_base_l}
    rotation {pitch=-5}
   }
   n45 = "Transform (Modify) Bone - Bone: dyn_skirt_fr_a_base_r"(ComponentPose=@n44, Alpha=@n43, Translation="0,0,0", Rotation="0,-5,-30", Scale="1,1,1") {
    rotationMode=BMM_Additive
    rotationSpace=BCS_BoneSpace
    translationMode=BMM_Additive
    translationSpace=BCS_BoneSpace
    alphaBoolBlend {blendOption=Linear}
    alphaScaleBiasClamp {bMapRange=true, inRange {max=0, min=-100}, outRange {max=0, min=1}}
    boneToModify {boneName=dyn_skirt_fr_a_base_r}
    rotation {roll=-30, yaw=-5}
   }
   n46 = "Transform (Modify) Bone - Bone: dyn_skirt_fr_a_base"(ComponentPose=@n45, Alpha=@n43, Translation="0,0,0", Rotation="0,10,0", Scale="1,1,1") {
    rotationMode=BMM_Additive
    rotationSpace=BCS_BoneSpace
    translationMode=BMM_Additive
    translationSpace=BCS_BoneSpace
    alphaBoolBlend {blendOption=Linear}
    alphaScaleBiasClamp {bMapRange=true, inRange {max=0, min=-100}, outRange {max=0, min=1}}
    boneToModify {boneName=dyn_skirt_fr_a_base}
    rotation {yaw=10}
   }
   n47 = "Transform (Modify) Bone - Bone: dyn_skirt_bk_a_base_l"(ComponentPose=@n46, Alpha=@n43, Translation="-3,5,0", Rotation="0,0,0", Scale="1,1,1") {
    rotationMode=BMM_Additive
    rotationSpace=BCS_BoneSpace
    translationMode=BMM_Additive
    translationSpace=BCS_BoneSpace
    alphaBoolBlend {blendOption=Linear}
    alphaScaleBiasClamp {bMapRange=true, inRange {max=0, min=-100}, outRange {max=0, min=1}}
    boneToModify {boneName=dyn_skirt_bk_a_base_l}
    translation=<-3 5>
   }
   n48 = Get ThighRotationRight()
   n49 = "Transform (Modify) Bone - Bone: dyn_skirt_fr_a_base_l"(ComponentPose=@n47, Alpha=@n48, Translation="0,0,0", Rotation="0,-5,-30", Scale="1,1,1") {
    rotationMode=BMM_Additive
    rotationSpace=BCS_BoneSpace
    translationMode=BMM_Additive
    translationSpace=BCS_BoneSpace
    alphaBoolBlend {blendOption=Linear}
    alphaScaleBiasClamp {bMapRange=true, inRange {max=0, min=-100}, outRange {max=0, min=1}}
    boneToModify {boneName=dyn_skirt_fr_a_base_l}
    rotation {roll=-30, yaw=-5}
   }
   n50 = "Transform (Modify) Bone - Bone: dyn_skirt_fr_a_base_r"(ComponentPose=@n49, Alpha=@n48, Translation="0,0,0", Rotation="-5,0,0", Scale="1,1,1") {
    rotationMode=BMM_Additive
    rotationSpace=BCS_BoneSpace
    translationMode=BMM_Additive
    translationSpace=BCS_BoneSpace
    alphaBoolBlend {blendOption=Linear}
    alphaScaleBiasClamp {bMapRange=true, inRange {max=0, min=-100}, outRange {max=0, min=1}}
    boneToModify {boneName=dyn_skirt_fr_a_base_r}
    rotation {pitch=-5}
   }
   n51 = "Transform (Modify) Bone - Bone: dyn_skirt_fr_a_base"(ComponentPose=@n50, Alpha=@n48, Translation="0,0,0", Rotation="0,10,0", Scale="1,1,1") {
    rotationMode=BMM_Additive
    rotationSpace=BCS_BoneSpace
    translationMode=BMM_Additive
    translationSpace=BCS_BoneSpace
    alphaBoolBlend {blendOption=Linear}
    alphaScaleBiasClamp {bMapRange=true, inRange {max=0, min=-100}, outRange {max=0, min=1}}
    boneToModify {boneName=dyn_skirt_fr_a_base}
    rotation {yaw=10}
   }
   n52 = "Transform (Modify) Bone - Bone: dyn_skirt_bk_a_base_r"(ComponentPose=@n51, Alpha=@n48, Translation="3,-5,0", Rotation="0,0,0", Scale="1,1,1") {
    rotationMode=BMM_Additive
    rotationSpace=BCS_BoneSpace
    translationMode=BMM_Additive
    translationSpace=BCS_BoneSpace
    alphaBoolBlend {blendOption=Linear}
    alphaScaleBiasClamp {bMapRange=true, inRange {max=0, min=-100}, outRange {max=0, min=1}}
    boneToModify {boneName=dyn_skirt_bk_a_base_r}
    translation=<3 -5>
   }
   n53 = "Transform (Modify) Bone - Bone: dyn_skirt_bk_a_base_l"(ComponentPose=@n52, Alpha=Get ThighRotationLeft(), Translation="0,0,0", Rotation="0,25,0", Scale="1,1,1") {
    rotationMode=BMM_Additive
    rotationSpace=BCS_BoneSpace
    translationMode=BMM_Additive
    translationSpace=BCS_BoneSpace
    alphaBoolBlend {blendOption=Linear}
    alphaScaleBiasClamp {bMapRange=true, inRange {max=15}}
    boneToModify {boneName=dyn_skirt_bk_a_base_l}
    rotation {yaw=25}
   }
   n54 = "Transform (Modify) Bone - Bone: dyn_skirt_bk_a_base_r"(ComponentPose=@n53, Alpha=Get ThighRotationRight(), Translation="0,0,0", Rotation="0,25,0", Scale="1,1,1") {
    rotationMode=BMM_Additive
    rotationSpace=BCS_BoneSpace
    translationMode=BMM_Additive
    translationSpace=BCS_BoneSpace
    alphaBoolBlend {blendOption=Linear}
    alphaScaleBiasClamp {bMapRange=true, inRange {max=15}}
    boneToModify {boneName=dyn_skirt_bk_a_base_r}
    rotation {yaw=25}
   }
   n55 = Get IsCrouchingAndStillAlpha()
   n56 = "Transform (Modify) Bone - Bone: dyn_skirt_bk_a_base_r"(ComponentPose=@n54, Alpha=@n55, Translation="4.5,0,0", Rotation="0,0,0", Scale="0.6,1,1") {
    rotationMode=BMM_Additive
    rotationSpace=BCS_BoneSpace
    scaleMode=BMM_Additive
    scaleSpace=BCS_BoneSpace
    translationMode=BMM_Additive
    translationSpace=BCS_BoneSpace
    alphaBoolBlend {blendOption=Linear}
    alphaScaleBiasClamp {inRange {max=0, min=-100}, outRange {max=0, min=1}}
    boneToModify {boneName=dyn_skirt_bk_a_base_r}
    scale {x=0.6}
    translation {x=4.5}
   }
   n57 = "Transform (Modify) Bone - Bone: dyn_skirt_thigh_r"(ComponentPose=@n56, Alpha=@n55, Translation="-2,0,0", Rotation="5,0,0", Scale="1,1,1") {
    rotationMode=BMM_Additive
    rotationSpace=BCS_BoneSpace
    scaleMode=BMM_Additive
    scaleSpace=BCS_BoneSpace
    translationMode=BMM_Additive
    translationSpace=BCS_BoneSpace
    alphaBoolBlend {blendOption=Linear}
    alphaScaleBiasClamp {inRange {max=0, min=-100}, outRange {max=0, min=1}}
    boneToModify {boneName=dyn_skirt_thigh_r}
    rotation {pitch=5}
    translation {x=-2}
   }
   n58 = Anim Dynamics - Bone: dyn_collar_fr_b_l(ComponentPose=@n57, Alpha=1, ExternalForce="0,0,0") {
    angularDampingOverride=1
    angularSpringConstant=1500
    bAngularSpring=true
    bGravityOverrideInSimSpace=true
    bOverrideAngularDamping=true
    gravityScale=0.35
    alphaBoolBlend {blendOption=Linear}
    boundBone {boneName=dyn_collar_fr_b_l}
    physicsBodyDefinitions [{
      sphereCollisionRadius=0
      boundBone {boneName=dyn_collar_fr_b_l}
      boxExtents=<10 20 20>
      constraintSetup {angularLimitsMax=<0 0 0>, angularLimitsMin=<0 -20 0>, angularTarget=<0 0 0>}
      localJointOffset=<10 0 0>
     }]
   }
   n59 = Anim Dynamics - Bone: dyn_collar_fr_b_r(ComponentPose=@n58, Alpha=1, ExternalForce="0,0,0") {
    angularDampingOverride=1
    angularSpringConstant=1500
    bAngularSpring=true
    bGravityOverrideInSimSpace=true
    bOverrideAngularDamping=true
    gravityScale=0.35
    alphaBoolBlend {blendOption=Linear}
    boundBone {boneName=dyn_collar_fr_b_r}
    physicsBodyDefinitions [{
      sphereCollisionRadius=0
      boundBone {boneName=dyn_collar_fr_b_r}
      boxExtents=<10 20 20>
      constraintSetup {angularLimitsMax=<0 0 0>, angularLimitsMin=<0 -20 0>, angularTarget=<0 0 0>}
      localJointOffset=<-10 0 0>
     }]
   }
   n60 = RigidBody(ComponentPose=@n59, Alpha=@n20) {
    bOverrideWorldGravity=true
    overridePhysicsAsset=&/F_MED_HighTower_Honeydew_L2_Physics
    simulationSpace=BaseBoneSpace
    alphaBoolBlend {blendOption=Linear}
    baseBoneRef {boneName=thigh_l}
    componentAppliedLinearAccClamp=<3000 3000 3000>
    componentLinearAccScale=<0.3 0.3 0.3>
    componentLinearVelScale=<0.3 0.3 0.3>
    overrideWorldGravity {z=-980}
   }
   n61 = RigidBody(ComponentPose=@n60, Alpha=@n20) {
    bOverrideWorldGravity=true
    overridePhysicsAsset=&/F_MED_HighTower_Honeydew_R2_Physics
    simulationSpace=BaseBoneSpace
    alphaBoolBlend {blendOption=Linear}
    baseBoneRef {boneName=thigh_r}
    componentAppliedLinearAccClamp=<3000 3000 3000>
    componentLinearAccScale=<0.3 0.3 0.3>
    componentLinearVelScale=<0.3 0.3 0.3>
    overrideWorldGravity {z=-980}
   }
   n62 = Get isFrontEndLockerPose() -> (isFrontEndLockerPose:double)
   n63 = "Transform (Modify) Bone - Bone: dyn_skirt_fr_a_r"(ComponentPose=@n61, Alpha=@n62, Translation="0,0,0", Rotation="0.000,0.000,20", Scale="1,1,1") {
    rotationMode=BMM_Additive
    rotationSpace=BCS_BoneSpace
    translationMode=BMM_Additive
    translationSpace=BCS_BoneSpace
    alphaBoolBlend {blendOption=Linear}
    alphaScaleBiasClamp {inRange {max=0, min=-100}, outRange {max=0, min=1}}
    boneToModify {boneName=dyn_skirt_fr_a_r}
    rotation {roll=20}
   }
   n64 = "Transform (Modify) Bone - Bone: dyn_skirt_fr_a_l"(ComponentPose=@n63, Alpha=@n62, Translation="0,0,0", Rotation="-15.000,0.000,-15", Scale="0.86,1,1") {
    rotationMode=BMM_Additive
    rotationSpace=BCS_BoneSpace
    scaleMode=BMM_Additive
    scaleSpace=BCS_BoneSpace
    translationMode=BMM_Additive
    translationSpace=BCS_BoneSpace
    alphaBoolBlend {blendOption=Linear}
    alphaScaleBiasClamp {inRange {max=0, min=-100}, outRange {max=0, min=1}}
    boneToModify {boneName=dyn_skirt_fr_a_l}
    rotation {pitch=-15, roll=-15}
    scale {x=0.86}
   }
   n65 = "Transform (Modify) Bone - Bone: dyn_skirt_fr_a"(ComponentPose=@n64, Alpha=@n62, Translation="0,0,0", Rotation="10,5,-20", Scale="0.95,1,1") {
    rotationMode=BMM_Additive
    rotationSpace=BCS_BoneSpace
    scaleMode=BMM_Additive
    scaleSpace=BCS_BoneSpace
    translationMode=BMM_Additive
    translationSpace=BCS_BoneSpace
    alphaBoolBlend {blendOption=Linear}
    alphaScaleBiasClamp {inRange {max=0, min=-100}, outRange {max=0, min=1}}
    boneToModify {boneName=dyn_skirt_fr_a}
    rotation=<rot 10 5 -20>
    scale {x=0.95}
   }
   n66 = Get isFrontEndLockerPose()
   n67 = "Transform (Modify) Bone - Bone: dyn_skirt_bk_a_l"(ComponentPose=@n65, Alpha=@n66, Translation="0,0,0", Rotation="0.000,5,-20", Scale="1,1,1") {
    rotationMode=BMM_Additive
    rotationSpace=BCS_BoneSpace
    translationMode=BMM_Additive
    translationSpace=BCS_BoneSpace
    alphaBoolBlend {blendOption=Linear}
    alphaScaleBiasClamp {inRange {max=0, min=-100}, outRange {max=0, min=1}}
    boneToModify {boneName=dyn_skirt_bk_a_l}
    rotation {roll=-20, yaw=5}
   }
   n68 = "Transform (Modify) Bone - Bone: dyn_skirt_bk_a_r"(ComponentPose=@n67, Alpha=@n66, Translation="0,0,0", Rotation="-5,0,20", Scale="1,1,1") {
    rotationMode=BMM_Additive
    rotationSpace=BCS_BoneSpace
    translationMode=BMM_Additive
    translationSpace=BCS_BoneSpace
    alphaBoolBlend {blendOption=Linear}
    alphaScaleBiasClamp {inRange {max=0, min=-100}, outRange {max=0, min=1}}
    boneToModify {boneName=dyn_skirt_bk_a_r}
    rotation {pitch=-5, roll=20}
   }
   n69 = "Transform (Modify) Bone - Bone: dyn_skirt_bk_a"(ComponentPose=@n68, Alpha=@n66, Translation="0,0,0", Rotation="0,0,-10", Scale="0.95,1,1") {
    rotationMode=BMM_Additive
    rotationSpace=BCS_BoneSpace
    scaleMode=BMM_Additive
    scaleSpace=BCS_BoneSpace
    translationMode=BMM_Additive
    translationSpace=BCS_BoneSpace
    alphaBoolBlend {blendOption=Linear}
    alphaScaleBiasClamp {inRange {max=0, min=-100}, outRange {max=0, min=1}}
    boneToModify {boneName=dyn_skirt_bk_a}
    rotation {roll=-10}
    scale {x=0.95}
   }
   n70 = "Transform (Modify) Bone - Bone: dyn_skirt_bk_a_l"!(ComponentPose=@n69, bAlphaBoolEnabled:bool=True, Translation="0,0,3", Rotation="0,5,0", Scale="1,1,1") -> (Pose) {
    alphaInputType=Bool
    rotationMode=BMM_Additive
    rotationSpace=BCS_BoneSpace
    translationMode=BMM_Additive
    translationSpace=BCS_BoneSpace
    alphaBoolBlend {blendInTime=0.2, blendOption=Linear, blendOutTime=0.2}
    alphaScaleBiasClamp {inRange {max=0, min=-100}, outRange {max=0, min=1}}
    boneToModify {boneName=dyn_skirt_bk_a_l}
    rotation {yaw=5}
    translation {z=3}
   }
   n71 = "Transform (Modify) Bone - Bone: dyn_skirt_bk_a_r"!(ComponentPose=@n70, bAlphaBoolEnabled:bool=True, Translation="0,0,-3", Rotation="0,5,0", Scale="1,1,1") -> (Pose) {
    alphaInputType=Bool
    rotationMode=BMM_Additive
    rotationSpace=BCS_BoneSpace
    translationMode=BMM_Additive
    translationSpace=BCS_BoneSpace
    alphaBoolBlend {blendInTime=0.2, blendOption=Linear, blendOutTime=0.2}
    alphaScaleBiasClamp {inRange {max=0, min=-100}, outRange {max=0, min=1}}
    boneToModify {boneName=dyn_skirt_bk_a_r}
    rotation {yaw=5}
    translation {z=-3}
   }
   n72 = "Transform (Modify) Bone - Bone: dyn_skirt_bk_a"!(ComponentPose=@n71, bAlphaBoolEnabled:bool=True, Translation="0,0,0", Rotation="0,5,0", Scale="1,1,1") -> (Pose) {
    alphaInputType=Bool
    rotationMode=BMM_Additive
    rotationSpace=BCS_BoneSpace
    scaleMode=BMM_Additive
    scaleSpace=BCS_BoneSpace
    translationMode=BMM_Additive
    translationSpace=BCS_BoneSpace
    alphaBoolBlend {blendInTime=0.2, blendOption=Linear, blendOutTime=0.2}
    alphaScaleBiasClamp {inRange {max=0, min=-100}, outRange {max=0, min=1}}
    boneToModify {boneName=dyn_skirt_bk_a}
    rotation {yaw=5}
   }
   n73 = "Transform (Modify) Bone - Bone: dyn_skirt_fr_a_r"!(ComponentPose=@n72, bAlphaBoolEnabled:bool=True, Translation="0,0,0", Rotation="-0.000,-5.000,-0.000", Scale="1,1,1") -> (Pose) {
    alphaInputType=Bool
    rotationMode=BMM_Additive
    rotationSpace=BCS_BoneSpace
    translationMode=BMM_Additive
    translationSpace=BCS_BoneSpace
    alphaBoolBlend {blendInTime=0.2, blendOption=Linear, blendOutTime=0.2}
    alphaScaleBiasClamp {inRange {max=0, min=-100}, outRange {max=0, min=1}}
    boneToModify {boneName=dyn_skirt_fr_a_r}
    rotation=<rot 0 -5 0>
   }
   n74 = "Transform (Modify) Bone - Bone: dyn_skirt_fr_a_l"!(ComponentPose=@n73, bAlphaBoolEnabled:bool=True, Translation="0,0,0", Rotation="-0.000,-5.000,-0.000", Scale="1,1,1") -> (Pose) {
    alphaInputType=Bool
    rotationMode=BMM_Additive
    rotationSpace=BCS_BoneSpace
    scaleMode=BMM_Additive
    scaleSpace=BCS_BoneSpace
    translationMode=BMM_Additive
    translationSpace=BCS_BoneSpace
    alphaBoolBlend {blendInTime=0.2, blendOption=Linear, blendOutTime=0.2}
    alphaScaleBiasClamp {inRange {max=0, min=-100}, outRange {max=0, min=1}}
    boneToModify {boneName=dyn_skirt_fr_a_l}
    rotation=<rot 0 -5.0001 0>
   }
   n75 = "Transform (Modify) Bone - Bone: dyn_skirt_fr_a"!(ComponentPose=@n74, bAlphaBoolEnabled:bool=True, Translation="0,0,0", Rotation="0.000,5.000,0.000", Scale="1,1,1") -> (Pose) {
    alphaInputType=Bool
    rotationMode=BMM_Additive
    rotationSpace=BCS_BoneSpace
    scaleMode=BMM_Additive
    scaleSpace=BCS_BoneSpace
    translationMode=BMM_Additive
    translationSpace=BCS_BoneSpace
    alphaBoolBlend {blendInTime=0.2, blendOption=Linear, blendOutTime=0.2}
    alphaScaleBiasClamp {inRange {max=0, min=-100}, outRange {max=0, min=1}}
    boneToModify {boneName=dyn_skirt_fr_a}
    rotation=<rot 0 5 0>
   }
   Output Pose(Result=Component To Local(ComponentPose=@n75)) {layerGroup=None}
  ]}
 EventGraph {nodes [
   Event BlueprintUpdateAnimation()
   n1 = Get Anim_Inst() -> (Anim_Inst:Anim Instance Object Reference)
   > Set isFrontEndLockerPose(isFrontEndLockerPose:double=SelectFloat(
     A=1
     B=0
     bPickA=Montage_IsActive!(
      self:Anim Instance Object Reference=@n1
      Montage:Anim Montage Object Reference=/BRCosmetics/Animation/Game/MainPlayer/Menu/BR/CustomSkins/_S14/HighTower/HighTower_HoneyDew_Female_Idle_Montage
     ) -> (ReturnValue:bool)
    )) -> (Output_Get:double)
   Event BlueprintInitializeAnimation()
   > n2 = Set Skel_Mesh(Skel_Mesh=GetOwnerSkeletalMeshComponent())
   > Is Valid(InputObject=@n2)
   > Set Anim_Inst(Anim_Inst:Anim Instance Object Reference=GetAnimInstance(self=@n2)) -> (Output_Get:Anim Instance Object Reference)
   Reroute Node(InputPin:Wildcard) -> (OutputPin:Wildcard)
  ]}
]
variables [
 Anim_Inst {type=AnimInstance, default=null}
 isFrontEndLockerPose {type=number}
 Skel_Mesh {type=SkeletalMeshComponent}
]
