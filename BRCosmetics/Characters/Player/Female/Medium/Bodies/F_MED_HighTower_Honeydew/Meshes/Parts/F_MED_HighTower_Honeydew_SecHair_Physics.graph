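# Physics asset for the secondary hair and earrings of the HighTower Honeydew medium female player body.
# A tapering capsule ponytail chain hangs from the head via a free-jointed base, with limited swing and twist down its length.
# Short, tightly limited branch chains, twist-locked at their roots, split off the first ponytail segment, while box earrings constrain to the head.
# Head, neck and upper-spine capsules form the body shapes, linked by free constraints.
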PhysicsAsset /BRCosmetics/Characters/Player/Female/Medium/Bodies/F_MED_HighTower_Honeydew/Meshes/Parts/F_MED_HighTower_Honeydew_SecHair_Physics
bodies {
 dyn_earring_l [{shapeType=Box, extentX=4.8581, extentY=0.5, extentZ=4.8043, center=<2.1697 -0.0689 -0.1528>, rotation=<rot 0 -92.3303 -86.3155>}]
 dyn_earring_r [{shapeType=Box, extentX=4.8581, extentY=0.5, extentZ=4.8043, center=<-2.1697 0.0689 0.1528>, rotation=<rot 0 -92.3303 93.6845>}]
 dyn_hair_branch_aa_l [{shapeType=Capsule, length=4.25, radius=1.25, center=<2.6809 0 0>, rotation=<rot 90 0 0>}]
 dyn_hair_branch_aa_r [{shapeType=Capsule, length=3.5, radius=1.25, center=<2.5735 0 0>, rotation=<rot 90 0 0>}]
 dyn_hair_branch_ab_l [{shapeType=Capsule, length=4, radius=1, center=<2.8129 0 0>, rotation=<rot 90 0 0>}]
 dyn_hair_branch_ab_r [{shapeType=Capsule, length=4, radius=1, center=<2.6857 0 0>, rotation=<rot 90 0 0>}]
 dyn_hair_branch_ac_l [{shapeType=Capsule, length=4.5, radius=0.75, center=<2.6382 0 -0.0002>, rotation=<rot 90 0 0>}]
 dyn_hair_branch_ba_r [{shapeType=Capsule, length=2.5, radius=1.25, center=<1.8928 0 0>, rotation=<rot 90 0 0>}]
 dyn_hair_branch_bb_r [{shapeType=Capsule, length=3.75, radius=1.25, center=<2.4286 0 0>, rotation=<rot 90 0 0>}]
 dyn_hair_branch_bc_r [{shapeType=Capsule, length=4, radius=0.75, center=<2.4169 -0.0756 0.0001>, rotation=<rot 90 20.556 25.6367>}]
 dyn_ponyTail_a [{shapeType=Capsule, length=0, radius=6.5, center=<2.8073 2.3759 0>, rotation=<rot 90 0 0>}]
 dyn_ponyTail_a_base [{shapeType=Capsule, length=1, radius=1}]
 dyn_ponyTail_b [{shapeType=Capsule, length=2.25, radius=4.75, center=<3.3828 2.0893 0>, rotation=<rot 88.408 13.6585 0>}]
 dyn_ponyTail_c [{shapeType=Capsule, length=3.25, radius=3.25, center=<2.4984 1.5831 -0.5573>, rotation=<rot 90 13.4732 14.8564>}]
 dyn_ponyTail_d [{shapeType=Capsule, length=4, radius=2, center=<2.5909 0.0483 0>, rotation=<rot 90 90 96.0812>}]
 dyn_ponyTail_e [{shapeType=Capsule, length=5.3438, radius=1.1182, center=<2.0712 -0.3591 -0.9621>, rotation=<rot 0 -99.3224 -62.6604>}]
 head [
  {shapeType=Capsule, length=7.5, radius=5.5, center=<3.3678 0.6799 0>, rotation=<rot 90 0 0>}
  {shapeType=Capsule, length=1, radius=7.5, center=<5.8755 0 0>}
 ]
 neck_01 [{shapeType=Capsule, length=5, radius=4.5, center=<1.9299 0.1803 0>, rotation=<rot 90 -90 -93.0614>}]
 spine_05 [{shapeType=Capsule, length=22.5, radius=9.25, center=<3.3872 1.5179 0.0001>}]
}
constraints [
 {
  bone1Name=dyn_ponyTail_a_base
  bone2Name=head
  swing1LimitDegrees=45
  swing1Motion=Free
  swing2LimitDegrees=45
  swing2Motion=Free
  twistLimitDegrees=45
  twistMotion=Free
 }
 {bone1Name=dyn_ponyTail_a, bone2Name=dyn_ponyTail_a_base, swing1LimitDegrees=35, swing2LimitDegrees=35, twistLimitDegrees=5}
 {bone1Name=dyn_ponyTail_b, bone2Name=dyn_ponyTail_a, swing1LimitDegrees=35, swing2LimitDegrees=35, twistLimitDegrees=5}
 {bone1Name=dyn_ponyTail_c, bone2Name=dyn_ponyTail_b, swing1LimitDegrees=35, swing2LimitDegrees=35, twistLimitDegrees=5}
 {bone1Name=dyn_ponyTail_d, bone2Name=dyn_ponyTail_c, swing1LimitDegrees=35, swing2LimitDegrees=35, twistLimitDegrees=5}
 {
  bone1Name=head
  bone2Name=neck_01
  swing1LimitDegrees=45
  swing1Motion=Free
  swing2LimitDegrees=45
  swing2Motion=Free
  twistLimitDegrees=45
  twistMotion=Free
 }
 {bone1Name=dyn_ponyTail_e, bone2Name=dyn_ponyTail_d, swing1LimitDegrees=20, swing2LimitDegrees=20, twistLimitDegrees=5}
 {
  bone1Name=neck_01
  bone2Name=spine_05
  swing1LimitDegrees=45
  swing1Motion=Free
  swing2LimitDegrees=45
  swing2Motion=Free
  twistLimitDegrees=45
  twistMotion=Free
 }
 {bone1Name=dyn_earring_l, bone2Name=head, swing1LimitDegrees=15, swing2LimitDegrees=30, twistLimitDegrees=15}
 {bone1Name=dyn_earring_r, bone2Name=head, swing1LimitDegrees=15, swing2LimitDegrees=30, twistLimitDegrees=15}
 {
  bone1Name=dyn_hair_branch_aa_r
  bone2Name=dyn_ponyTail_a
  swing1LimitDegrees=5
  swing2LimitDegrees=5
  swing2Motion=Locked
  twistLimitDegrees=5
  twistMotion=Locked
 }
 {bone1Name=dyn_hair_branch_ab_r, bone2Name=dyn_hair_branch_aa_r, swing1LimitDegrees=10, swing2LimitDegrees=10, twistLimitDegrees=5}
 {
  bone1Name=dyn_hair_branch_ba_r
  bone2Name=dyn_ponyTail_a
  swing1LimitDegrees=10
  swing2LimitDegrees=10
  twistLimitDegrees=5
  twistMotion=Locked
 }
 {bone1Name=dyn_hair_branch_bb_r, bone2Name=dyn_hair_branch_ba_r, swing1LimitDegrees=10, swing2LimitDegrees=10, twistLimitDegrees=5}
 {bone1Name=dyn_hair_branch_bc_r, bone2Name=dyn_hair_branch_bb_r, swing1LimitDegrees=10, swing2LimitDegrees=10, twistLimitDegrees=5}
 {
  bone1Name=dyn_hair_branch_aa_l
  bone2Name=dyn_ponyTail_a
  swing1LimitDegrees=10
  swing2LimitDegrees=10
  twistLimitDegrees=5
  twistMotion=Locked
 }
 {bone1Name=dyn_hair_branch_ab_l, bone2Name=dyn_hair_branch_aa_l, swing1LimitDegrees=10, swing2LimitDegrees=10, twistLimitDegrees=5}
 {bone1Name=dyn_hair_branch_ac_l, bone2Name=dyn_hair_branch_ab_l, swing1LimitDegrees=10, swing2LimitDegrees=10, twistLimitDegrees=5}
]
